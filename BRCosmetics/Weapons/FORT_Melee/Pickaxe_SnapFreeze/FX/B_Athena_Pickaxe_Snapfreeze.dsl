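; Cosmetic Blueprint for the Snapfreeze pickaxe, driving its variant-aware swing and impact effects.
; Swing events check for the "Swing1" montage section to hide the anim trail and, outside frontend preview, spawn first-attack FX on the handle.
; In preview, impacts replace default FX with first-attack or impact Niagara systems, fed a cycling counter and debounced against duplicate signals.
; The particle variant comes from the pickaxe loadout and is cached on equip for previews.
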(event EventLegacyAnimTrailSetup
  (params (Activate :type Boolean)))

(event EventBeginPlay)

(event EventSwingRight
  (bind _returnvalue (Game|GetInstigator))
  (bind _asfort_pawn (Utilities|Casting|CastToFortPawn _returnvalue))
  (bind _mesh (Variables|/Script/Engine.Character|GetMesh _asfort_pawn))
  (bind _returnvalue_2 (Components|SkeletalMesh|GetAnimInstance _mesh))
  (bind _returnvalue_3 (Animation|Montage|GetCurrentActiveMontage _returnvalue_2))
  (bind _returnvalue_4 (Animation|Montage|MontageGetCurrentSection _returnvalue_2 _returnvalue_3))
  (bind _returnvalue_5 (== _returnvalue_4 "Swing1"))
  (|CallParentFunction)
  (Networking|CanExecuteCosmeticEvents
    (:True
      (Utilities|FlowControl|Sequence
        (:then_0
          (if _returnvalue_5
            (if (not (Reactive|Frontend|IsCosmeticObjectDisplayedasPreview 0))
              (bind _returnvalue_6 (Effects|SpawnFXSystemAttached (Variables|Default|GetFXFirstAttack) (Variables|Visual|GetWeaponMesh) "Melee_TwoHanded_Handle"))
              (bind _asfort_player_pawn (Utilities|Casting|CastToFortPlayerPawn (Variables|Actor|GetInstigator)))
              (bind _index (CosmeticLoadout|GetVariantChannelActiveIndexforSlotActor "(TagName=\"LoadoutSlot.Pickaxe\")" _asfort_player_pawn "(TagName=\"Cosmetics.Variant.Channel.Particle\")" (Utilities|Array|MakeArray "(TagName=\"Cosmetics.Variant.Property.Particle1\")" "(TagName=\"Cosmetics.Variant.Property.Particle2\")") false false))
              (Effects|Components|ParticleSystem|SetFloatParameter _returnvalue_6 "Variant" (Math|Conversions|ToFloat(Integer) _index)))))
        (:then_1
          (Development|SetHiddeninGame (CallFunction|GetAnimTrailVFXComponent) _returnvalue_5 false))))))

(event EventSwingLeft
  (bind _returnvalue (Game|GetInstigator))
  (bind _asfort_pawn (Utilities|Casting|CastToFortPawn _returnvalue))
  (bind _mesh (Variables|/Script/Engine.Character|GetMesh _asfort_pawn))
  (bind _returnvalue_2 (Components|SkeletalMesh|GetAnimInstance _mesh))
  (bind _returnvalue_3 (Animation|Montage|GetCurrentActiveMontage _returnvalue_2))
  (bind _returnvalue_4 (Animation|Montage|MontageGetCurrentSection _returnvalue_2 _returnvalue_3))
  (bind _returnvalue_5 (== _returnvalue_4 "Swing1"))
  (|CallParentFunction)
  (Networking|CanExecuteCosmeticEvents
    (:True
      (Utilities|FlowControl|Sequence
        (:then_0
          (if _returnvalue_5
            (if (not (Reactive|Frontend|IsCosmeticObjectDisplayedasPreview 0))
              (bind _returnvalue_6 (Effects|SpawnFXSystemAttached (Variables|Default|GetFXFirstAttack) (Variables|Visual|GetWeaponMesh) "Melee_TwoHanded_Handle"))
              (bind _asfort_player_pawn (Utilities|Casting|CastToFortPlayerPawn (Variables|Actor|GetInstigator)))
              (bind _index (CosmeticLoadout|GetVariantChannelActiveIndexforSlotActor "(TagName=\"LoadoutSlot.Pickaxe\")" _asfort_player_pawn "(TagName=\"Cosmetics.Variant.Channel.Particle\")" (Utilities|Array|MakeArray "(TagName=\"Cosmetics.Variant.Property.Particle1\")" "(TagName=\"Cosmetics.Variant.Property.Particle2\")") false false))
              (Effects|Components|ParticleSystem|SetFloatParameter _returnvalue_6 "Variant" (Math|Conversions|ToFloat(Integer) _index)))))
        (:then_1
          (Development|SetHiddeninGame (CallFunction|GetAnimTrailVFXComponent) _returnvalue_5 false))))))

(event FortWeapon|EventOnPlayImpactFX
  (params (HitResult :type Hit%20Result%20Structure%20%28by%20ref%29) (ImpactPhysicalSurface :type EPhysicalSurface%20Enum :default "SurfaceType_Default") (SpawnedPSC :type FXSystem%20Component%20Object%20Reference))
  (bind (_returnvalue_location _returnvalue_rotation _returnvalue_scale) (Math|Transform|BreakTransform (Transformation|GetActorTransform (Variables|Default|GetAs%20Fort%20Player%20Pawn))))
  (bind _as_fort_player_pawn (Variables|Default|GetAs%20Fort%20Player%20Pawn))
  (bind _mesh (Variables|/Script/Engine.Character|GetMesh _as_fort_player_pawn))
  (bind _returnvalue (Components|SkeletalMesh|GetAnimInstance _mesh))
  (bind _returnvalue_2 (Animation|Montage|GetCurrentActiveMontage _returnvalue))
  (bind _returnvalue_3 (Animation|Montage|MontageGetCurrentSection _returnvalue _returnvalue_2))
  (bind _returnvalue_4 (== _returnvalue_3 "Swing1"))
  ; FE Impac
  (|CallParentFunction HitResult ImpactPhysicalSurface SpawnedPSC)
  (if (Reactive|Frontend|IsCosmeticObjectDisplayedasPreview 0)
    (Components|Activation|Deactivate SpawnedPSC)
    (if _returnvalue_4
      (bind _returnvalue_5 (Effects|SpawnFXSystemAttached "/BRCosmetics/Weapons/FORT_Melee/Pickaxe_SnapFreeze/FX/NS_Pickaxe_Snapfreeze_FirstAttack.NS_Pickaxe_Snapfreeze_FirstAttack" (Variables|Visual|GetWeaponMesh) "Melee_TwoHanded_Handle"))
      (else
        (bind _returnvalue_6 (Niagara|SpawnSystematLocation "/BRCosmetics/Weapons/FORT_Melee/Pickaxe_SnapFreeze/FX/NS_Pickaxe_Snapfreeze_Impact.NS_Pickaxe_Snapfreeze_Impact" (Math|Transform|TransformLocation (Transformation|GetActorTransform (Variables|Default|GetAs%20Fort%20Player%20Pawn)) "100.000000,0.000000,50.000000") (Math|Rotator|CombineRotators _returnvalue_rotation "90.000000,90.000000,0.000000") _returnvalue_scale true false))))
    (Utilities|FlowControl|DoOnce false
      (:Completed
        (Effects|Components|ParticleSystem|SetIntParameter (select _returnvalue_4 (Utilities|Casting|CastToNiagaraComponent _returnvalue_5) _returnvalue_6) "SwingSectionIndex" (Math|Integer|%25%28Integer%29 (Variables|Default|GetFE_Counter) 4))
        (Effects|Components|ParticleSystem|SetBoolParameter (select _returnvalue_4 (Utilities|Casting|CastToNiagaraComponent _returnvalue_5) _returnvalue_6) "bIsFrontEndPreview" true)
        (Effects|Components|ParticleSystem|SetFloatParameter (select _returnvalue_4 (Utilities|Casting|CastToNiagaraComponent _returnvalue_5) _returnvalue_6) "Variant" (Math|Conversions|ToFloat(Integer) (Variables|Default|GetFE_VariantIndex)))
        (Components|Activation|Activate (select _returnvalue_4 (Utilities|Casting|CastToNiagaraComponent _returnvalue_5) _returnvalue_6) false)
        (Math|Integer|IncrementInt (Variables|Default|GetFE_Counter)
          (:%20%20
            (Utilities|FlowControl|RetriggerableDelay 0.25) ; prevents duplicated signal
          ))))))

(event FortWeapon|EventOnEquippedAndAttached
  ; init
  (|CallParentFunction)
  (bind _asfort_player_pawn (Utilities|Casting|CastToFortPlayerPawn (Variables|Actor|GetInstigator)))
  (Variables|Default|SetAs%20Fort%20Player%20Pawn _asfort_player_pawn)
  (bind _index (CosmeticLoadout|GetVariantChannelActiveIndexforSlotActor "(TagName=\"LoadoutSlot.Pickaxe\")" _asfort_player_pawn "(TagName=\"Cosmetics.Variant.Channel.Particle\")" (Utilities|Array|MakeArray "(TagName=\"Cosmetics.Variant.Property.Particle1\")" "(TagName=\"Cosmetics.Variant.Property.Particle2\")")))
  (Variables|Default|SetFE_VariantIndex _index))

(fn ConstructionScript
  (|CallParentFunction)
  (Components|Activation|SetAutoActivate (CallFunction|GetIdlePSC) false))
