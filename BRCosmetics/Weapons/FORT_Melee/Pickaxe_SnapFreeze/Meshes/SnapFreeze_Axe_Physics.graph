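# Physics asset for the SnapFreeze pickaxe cosmetic, giving capsule bodies and joints to a branching set of dynamic hair bones on the axe staff.
# A long capsule covers the staff, and a tiny base_hair capsule joins it with free swing and twist.
# From there dyn_hair branches into left, middle and right sections with tight swing limits and mostly locked twist.
# Both left and middle tip pairs attach directly to their branch root rather than chaining.
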PhysicsAsset /BRCosmetics/Weapons/FORT_Melee/Pickaxe_SnapFreeze/Meshes/SnapFreeze_Axe_Physics
bodies {
 base_hair [{shapeType=Capsule, length=0.505, radius=0.505, shapeName=base_hair_capsule, rotation=<rot 0 -90 0>}]
 dyn_hair [{shapeType=Capsule, length=5.9444, radius=2.3719, shapeName=dyn_hair_capsule, center=<2.6047 -0.5971 0>}]
 dyn_hair_l [{shapeType=Capsule, length=4.3056, radius=1.8726, shapeName=dyn_hair_l_capsule, center=<1.8511 0 0>}]
 dyn_hair_l_01 [{shapeType=Capsule, length=3.3557, radius=1.5979, shapeName=dyn_hair_l_01_capsule, center=<3.0747 0 0>, rotation=<rot 90 0 0>}]
 dyn_hair_l_02 [{shapeType=Capsule, length=3.4246, radius=1.4041, shapeName=dyn_hair_l_02_capsule, center=<2.4414 0 0>, rotation=<rot 90 0 0>}]
 dyn_hair_m [{shapeType=Capsule, length=4.3056, radius=1.8726, shapeName=dyn_hair_l_capsule, center=<1.8511 0 0>}]
 dyn_hair_m_01 [{shapeType=Capsule, length=4.2048, radius=1.2309, shapeName=dyn_hair_m_01_capsule, center=<3.3467 0 0>, rotation=<rot 90 0 0>}]
 dyn_hair_m_02 [{shapeType=Capsule, length=2.2303, radius=1.1965, shapeName=dyn_hair_m_02_capsule, center=<2.4918 0 0>, rotation=<rot 90 0 0>}]
 dyn_hair_r [{shapeType=Capsule, length=2.9559, radius=1.0033, shapeName=dyn_hair_r_capsule, center=<2.24 0 0>, rotation=<rot 90 0 0>}]
 dyn_hair_r_01 [{shapeType=Capsule, length=2.1753, radius=1, shapeName=dyn_hair_r_01_capsule, center=<1.8581 0.0981 0.4079>, rotation=<rot 90 0 0>}]
 staff [{shapeType=Capsule, length=110.1154, radius=1.5744, center=<-5.647 -0.1772 -0.0709>, rotation=<rot 0 -89.8915 -89.9888>}]
}
constraints [
 {
  bone1Name=base_hair
  bone2Name=staff
  swing1LimitDegrees=45
  swing1Motion=Free
  swing2LimitDegrees=45
  swing2Motion=Free
  twistLimitDegrees=45
  twistMotion=Free
 }
 {bone1Name=dyn_hair, bone2Name=base_hair, swing1LimitDegrees=15, swing2LimitDegrees=15, twistLimitDegrees=45, twistMotion=Locked}
 {bone1Name=dyn_hair_l, bone2Name=dyn_hair, swing1LimitDegrees=25, swing2LimitDegrees=20, twistLimitDegrees=45, twistMotion=Locked}
 {bone1Name=dyn_hair_l_01, bone2Name=dyn_hair_l, swing1LimitDegrees=15, swing2LimitDegrees=35, twistLimitDegrees=5}
 {bone1Name=dyn_hair_l_02, bone2Name=dyn_hair_l, swing1LimitDegrees=15, swing2LimitDegrees=35, twistLimitDegrees=5}
 {bone1Name=dyn_hair_m, bone2Name=dyn_hair, swing1LimitDegrees=25, swing2LimitDegrees=20, twistLimitDegrees=45, twistMotion=Locked}
 {bone1Name=dyn_hair_m_01, bone2Name=dyn_hair_m, swing1LimitDegrees=15, swing2LimitDegrees=35, twistLimitDegrees=5}
 {bone1Name=dyn_hair_m_02, bone2Name=dyn_hair_m, swing1LimitDegrees=15, swing2LimitDegrees=35, twistLimitDegrees=5}
 {bone1Name=dyn_hair_r, bone2Name=dyn_hair, swing1LimitDegrees=15, swing2LimitDegrees=15, twistLimitDegrees=45, twistMotion=Locked}
 {bone1Name=dyn_hair_r_01, bone2Name=dyn_hair_r, swing1LimitDegrees=15, swing2LimitDegrees=35, twistLimitDegrees=45, twistMotion=Locked}
]
